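# Post-copy Anim Blueprint for the male medium Farmer body cosmetic that adds secondary motion to its "mstr_hay" bones.
# Copies the attached parent's pose, then weighted Constraint nodes tie most hay bones to nearby spine, clavicle and arm bones.
# A chain of single-body AnimDynamics springs then layers low-alpha, angle-limited sway on all driven hay bones plus a few unconstrained ones, with LOD threshold 1.
# The event graph has no logic.
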AnimBlueprint /BRCosmetics/Characters/Player/Male/Medium/Bodies/M_MED_Farmer/Meshes/M_MED_Farmer_Body_AnimBP
target_skeleton=&/M_MED_Farmer_Body_Skeleton
graphs [
 AnimGraph {nodes [
   n1 = "Constraint - mstr_hay_14 (PPPPPPPPPP)"!(
    ComponentPose=Local To Component(LocalPose=Copy Pose From Mesh() {bUseAttachedParent=true})
    Alpha=1
    ConstraintWeights_0=0.061
    ConstraintWeights_1:float=0.6428
    ConstraintWeights_2:float=0
    ConstraintWeights_3:float=0.0898
    ConstraintWeights_4:float=0
    ConstraintWeights_5:float=0.0016
    ConstraintWeights_6:float=0.0062
    ConstraintWeights_7:float=0
    ConstraintWeights_8:float=0
    ConstraintWeights_9:float=0.1985
   ) -> (Pose) {
    alphaBoolBlend {blendOption=Linear}
    boneToModify {boneName=mstr_hay_14}
    constraintSetup [
     {targetBone {boneName=upperarm_r}}
     {targetBone {boneName=upperarm_twist_01_r}}
     {targetBone {boneName=thigh_twist_01_r}}
     {targetBone {boneName=upperarm_twist_02_r}}
     {targetBone {boneName=spine_01}}
     {targetBone {boneName=spine_02}}
     {targetBone {boneName=spine_03}}
     {targetBone {boneName=spine_04}}
     {targetBone {boneName=spine_05}}
     {targetBone {boneName=clavicle_r}}
    ]
    constraintWeights [1,1,1,1,1,1,1,1,1,1]
   }
   n2 = "Constraint - mstr_hay_15 (PPPPPPPPPP)"!(
    ComponentPose=@n1
    Alpha=1
    ConstraintWeights_0=0.0267
    ConstraintWeights_1:float=0.5073
    ConstraintWeights_2:float=0
    ConstraintWeights_3:float=0.0341
    ConstraintWeights_4:float=0
    ConstraintWeights_5:float=0.0089
    ConstraintWeights_6:float=0.043
    ConstraintWeights_7:float=0.0028
    ConstraintWeights_8:float=0
    ConstraintWeights_9:float=0.3771
   ) -> (Pose) {
    alphaBoolBlend {blendOption=Linear}
    boneToModify {boneName=mstr_hay_15}
    constraintSetup [
     {targetBone {boneName=upperarm_r}}
     {targetBone {boneName=upperarm_twist_01_r}}
     {targetBone {boneName=thigh_twist_01_r}}
     {targetBone {boneName=upperarm_twist_02_r}}
     {targetBone {boneName=spine_01}}
     {targetBone {boneName=spine_02}}
     {targetBone {boneName=spine_03}}
     {targetBone {boneName=spine_04}}
     {targetBone {boneName=spine_05}}
     {targetBone {boneName=clavicle_r}}
    ]
    constraintWeights [1,1,1,1,1,1,1,1,1,1]
   }
   n3 = "Constraint - mstr_hay_13 (PPPPPPP)"!(
    ComponentPose=@n2
    Alpha=1
    ConstraintWeights_0=0.0834
    ConstraintWeights_1:float=0.5477
    ConstraintWeights_2:float=0.2774
    ConstraintWeights_3:float=0.001
    ConstraintWeights_4:float=0.0037
    ConstraintWeights_5:float=0.0001
    ConstraintWeights_6:float=0.0866
   ) -> (Pose) {
    alphaBoolBlend {blendOption=Linear}
    boneToModify {boneName=mstr_hay_13}
    constraintSetup [
     {targetBone {boneName=upperarm_r}}
     {targetBone {boneName=upperarm_twist_01_r}}
     {targetBone {boneName=upperarm_twist_02_r}}
     {targetBone {boneName=spine_02}}
     {targetBone {boneName=spine_03}}
     {targetBone {boneName=spine_04}}
     {targetBone {boneName=clavicle_r}}
    ]
    constraintWeights [1,1,1,1,1,1,1]
   }
   n4 = "Constraint - mstr_hay_16 (PPPPPPPP)"!(
    ComponentPose=@n3
    Alpha=1
    ConstraintWeights_0=0.0903
    ConstraintWeights_1:float=0.4415
    ConstraintWeights_2:float=0.3297
    ConstraintWeights_3:float=0
    ConstraintWeights_4:float=0.0055
    ConstraintWeights_5:float=0.0167
    ConstraintWeights_6:float=0.0007
    ConstraintWeights_7:float=0.1156
   ) -> (Pose) {
    alphaBoolBlend {blendOption=Linear}
    boneToModify {boneName=mstr_hay_16}
    constraintSetup [
     {targetBone {boneName=upperarm_r}}
     {targetBone {boneName=upperarm_twist_01_r}}
     {targetBone {boneName=upperarm_twist_02_r}}
     {targetBone {boneName=spine_01}}
     {targetBone {boneName=spine_02}}
     {targetBone {boneName=spine_03}}
     {targetBone {boneName=spine_04}}
     {targetBone {boneName=clavicle_r}}
    ]
    constraintWeights [1,1,1,1,1,1,1,1]
   }
   n5 = "Constraint - mstr_hay_17 (PPP)"!(ComponentPose=@n4, Alpha=1, ConstraintWeights_0=0.2318, ConstraintWeights_1:float=0.4802, ConstraintWeights_2:float=0.288) -> (Pose) {
    alphaBoolBlend {blendOption=Linear}
    boneToModify {boneName=mstr_hay_17}
    constraintSetup [{targetBone {boneName=upperarm_twist_01_r}}, {targetBone {boneName=upperarm_r}}, {targetBone {boneName=upperarm_twist_02_r}}]
    constraintWeights [1,1,1]
   }
   n6 = "Constraint - mstr_hay_18 (PPPPPP)"!(
    ComponentPose=@n5
    Alpha=1
    ConstraintWeights_0=0
    ConstraintWeights_1:float=0.0287
    ConstraintWeights_2:float=0.0291
    ConstraintWeights_3:float=0.3783
    ConstraintWeights_4:float=0.1184
    ConstraintWeights_5:float=0.4456
   ) -> (Pose) {
    alphaBoolBlend {blendOption=Linear}
    boneToModify {boneName=mstr_hay_18}
    constraintSetup [
     {targetBone {boneName=upperarm_r}}
     {targetBone {boneName=clavicle_l}}
     {targetBone {boneName=spine_02}}
     {targetBone {boneName=spine_03}}
     {targetBone {boneName=spine_04}}
     {targetBone {boneName=clavicle_r}}
    ]
    constraintWeights [1,1,1,1,1,1]
   }
   n7 = "Constraint - mstr_hay_02 (PPPPP)"!(
    ComponentPose=@n6
    Alpha=1
    ConstraintWeights_0=0.2217
    ConstraintWeights_1:float=0.0593
    ConstraintWeights_2:float=0.1917
    ConstraintWeights_3:float=0.4583
    ConstraintWeights_4:float=0.0691
   ) -> (Pose) {
    alphaBoolBlend {blendOption=Linear}
    boneToModify {boneName=mstr_hay_02}
    constraintSetup [
     {targetBone {boneName=clavicle_r}}
     {targetBone {boneName=spine_02}}
     {targetBone {boneName=spine_03}}
     {targetBone {boneName=spine_04}}
     {targetBone {boneName=spine_05}}
    ]
    constraintWeights [1,1,1,1,1]
   }
   n8 = "Constraint - mstr_hay_010 (PPPPPP)"!(
    ComponentPose=@n7
    Alpha=1
    ConstraintWeights_0=0.1027
    ConstraintWeights_1:float=0.0024
    ConstraintWeights_2:float=0.1548
    ConstraintWeights_3:float=0.2828
    ConstraintWeights_4:float=0.4522
    ConstraintWeights_5:float=0.0052
   ) -> (Pose) {
    alphaBoolBlend {blendOption=Linear}
    boneToModify {boneName=mstr_hay_010}
    constraintSetup [
     {targetBone {boneName=clavicle_l}}
     {targetBone {boneName=spine_01}}
     {targetBone {boneName=spine_02}}
     {targetBone {boneName=spine_03}}
     {targetBone {boneName=spine_04}}
     {targetBone {boneName=spine_05}}
    ]
    constraintWeights [1,1,1,1,1,1]
   }
   n9 = "Constraint - mstr_hay_11 (PPPPPPPPPPP)"!(
    ComponentPose=@n8
    Alpha=1
    ConstraintWeights_0=0.4999
    ConstraintWeights_1:float=0.0566
    ConstraintWeights_2:float=0
    ConstraintWeights_3:float=0.0368
    ConstraintWeights_4:float=0.1227
    ConstraintWeights_5:float=0.0007
    ConstraintWeights_6:float=0
    ConstraintWeights_7:float=0.2634
    ConstraintWeights_8:float=0
    ConstraintWeights_9:float=0
    ConstraintWeights_10:float=0.0198
   ) -> (Pose) {
    alphaBoolBlend {blendOption=Linear}
    boneToModify {boneName=mstr_hay_11}
    constraintSetup [
     {targetBone {boneName=clavicle_l}}
     {targetBone {boneName=upperarm_twist_02_l}}
     {targetBone {boneName=spine_01}}
     {targetBone {boneName=spine_02}}
     {targetBone {boneName=spine_03}}
     {targetBone {boneName=spine_04}}
     {targetBone {boneName=spine_05}}
     {targetBone {boneName=upperarm_twist_01_l}}
     {targetBone {boneName=clavicle_r}}
     {targetBone {boneName=thigh_twist_01_l}}
     {targetBone {boneName=upperarm_l}}
    ]
    constraintWeights [1,1,1,1,1,1,1,1,1,1,1]
   }
   n10 = "Constraint - mstr_hay_07 (PPPPPPPPPP)"!(
    ComponentPose=@n9
    Alpha=1
    ConstraintWeights_0=0.5489
    ConstraintWeights_1:float=0.0381
    ConstraintWeights_2:float=0
    ConstraintWeights_3:float=0.0026
    ConstraintWeights_4:float=0.0148
    ConstraintWeights_5:float=0.0002
    ConstraintWeights_6:float=0
    ConstraintWeights_7:float=0.3814
    ConstraintWeights_8:float=0
    ConstraintWeights_9:float=0.014
   ) -> (Pose) {
    alphaBoolBlend {blendOption=Linear}
    boneToModify {boneName=mstr_hay_07}
    constraintSetup [
     {targetBone {boneName=clavicle_l}}
     {targetBone {boneName=upperarm_twist_02_l}}
     {targetBone {boneName=spine_01}}
     {targetBone {boneName=spine_02}}
     {targetBone {boneName=spine_03}}
     {targetBone {boneName=spine_04}}
     {targetBone {boneName=spine_05}}
     {targetBone {boneName=upperarm_twist_01_l}}
     {targetBone {boneName=thigh_twist_01_l}}
     {targetBone {boneName=upperarm_l}}
    ]
    constraintWeights [1,1,1,1,1,1,1,1,1,1]
   }
   n11 = "Constraint - mstr_hay_10 (PPP)"!(ComponentPose=@n10, Alpha=1, ConstraintWeights_0=0.0073, ConstraintWeights_1:float=0.1366, ConstraintWeights_2:float=0.8561) -> (Pose) {
    alphaBoolBlend {blendOption=Linear}
    boneToModify {boneName=mstr_hay_10}
    constraintSetup [{targetBone {boneName=spine_04}}, {targetBone {boneName=upperarm_twist_01_l}}, {targetBone {boneName=clavicle_l}}]
    constraintWeights [1,1,1]
   }
   n12 = "Constraint - mstr_hay_08 (PPPPPPP)"!(
    ComponentPose=@n11
    Alpha=1
    ConstraintWeights_0=0.4978
    ConstraintWeights_1:float=0.0525
    ConstraintWeights_2:float=0.0006
    ConstraintWeights_3:float=0.0042
    ConstraintWeights_4:float=0.0002
    ConstraintWeights_5:float=0.426
    ConstraintWeights_6:float=0.0187
   ) -> (Pose) {
    alphaBoolBlend {blendOption=Linear}
    boneToModify {boneName=mstr_hay_08}
    constraintSetup [
     {targetBone {boneName=clavicle_l}}
     {targetBone {boneName=upperarm_twist_02_l}}
     {targetBone {boneName=spine_02}}
     {targetBone {boneName=spine_03}}
     {targetBone {boneName=spine_04}}
     {targetBone {boneName=upperarm_twist_01_l}}
     {targetBone {boneName=upperarm_l}}
    ]
    constraintWeights [1,1,1,1,1,1,1]
   }
   n13 = "Constraint - mstr_hay_09 (PPPPPPPPP)"!(
    ComponentPose=@n12
    Alpha=1
    ConstraintWeights_0=0.6428
    ConstraintWeights_1:float=0.0101
    ConstraintWeights_2:float=0
    ConstraintWeights_3:float=0.0005
    ConstraintWeights_4:float=0.0044
    ConstraintWeights_5:float=0.0009
    ConstraintWeights_6:float=0
    ConstraintWeights_7:float=0.3378
    ConstraintWeights_8:float=0.0034
   ) -> (Pose) {
    alphaBoolBlend {blendOption=Linear}
    boneToModify {boneName=mstr_hay_09}
    constraintSetup [
     {targetBone {boneName=clavicle_l}}
     {targetBone {boneName=upperarm_twist_02_l}}
     {targetBone {boneName=spine_01}}
     {targetBone {boneName=spine_02}}
     {targetBone {boneName=spine_03}}
     {targetBone {boneName=spine_04}}
     {targetBone {boneName=spine_05}}
     {targetBone {boneName=upperarm_twist_01_l}}
     {targetBone {boneName=upperarm_l}}
    ]
    constraintWeights [1,1,1,1,1,1,1,1,1]
   }
   n14 = "Constraint - mstr_hay_05 (PPPPPPPPPP)"!(
    ComponentPose=@n13
    Alpha=1
    ConstraintWeights_0=0.1722
    ConstraintWeights_1:float=0.2961
    ConstraintWeights_2:float=0
    ConstraintWeights_3:float=0.0039
    ConstraintWeights_4:float=0.0126
    ConstraintWeights_5:float=0.0008
    ConstraintWeights_6:float=0
    ConstraintWeights_7:float=0.4234
    ConstraintWeights_8:float=0
    ConstraintWeights_9:float=0.0911
   ) -> (Pose) {
    alphaBoolBlend {blendOption=Linear}
    boneToModify {boneName=mstr_hay_05}
    constraintSetup [
     {targetBone {boneName=clavicle_l}}
     {targetBone {boneName=upperarm_twist_02_l}}
     {targetBone {boneName=spine_01}}
     {targetBone {boneName=spine_02}}
     {targetBone {boneName=spine_03}}
     {targetBone {boneName=spine_04}}
     {targetBone {boneName=spine_05}}
     {targetBone {boneName=upperarm_twist_01_l}}
     {targetBone {boneName=thigh_twist_01_l}}
     {targetBone {boneName=upperarm_l}}
    ]
    constraintWeights [1,1,1,1,1,1,1,1,1,1]
   }
   n15 = "Constraint - mstr_hay_011 (PPP)"!(ComponentPose=@n14, Alpha=1, ConstraintWeights_0=0.4912, ConstraintWeights_1:float=0.2295, ConstraintWeights_2:float=0.2793) -> (Pose) {
    alphaBoolBlend {blendOption=Linear}
    boneToModify {boneName=mstr_hay_011}
    constraintSetup [{targetBone {boneName=upperarm_twist_02_l}}, {targetBone {boneName=upperarm_l}}, {targetBone {boneName=upperarm_twist_01_l}}]
    constraintWeights [1,1,1]
   }
   n16 = Anim Dynamics - Bone: mstr_hay_011(ComponentPose=@n15, Alpha=0.25, ExternalForce="0,0,0") {
    angularDampingOverride=0.96
    angularSpringConstant=35
    bAngularSpring=true
    bGravityOverrideInSimSpace=true
    bOverrideAngularDamping=true
    gravityScale=0.25
    lODThreshold=1
    linearDampingOverride=0.97
    numSolverIterationsPostUpdate=4
    numSolverIterationsPreUpdate=16
    simulationSpace=BoneRelative
    alphaBoolBlend {blendOption=Linear}
    boundBone {boneName=mstr_hay_011}
    componentLinearAccScale=<0.85 0.85 0.85>
    componentLinearVelScale=<0.85 0.85 0.85>
    physicsBodyDefinitions [{
      sphereCollisionRadius=0
      boundBone {boneName=mstr_hay_011}
      boxExtents=<5 5 5>
      constraintSetup {angularTargetAxis=AxisZ, twistAxis=AxisZ, angularLimitsMax=<30 30 0>, angularLimitsMin=<-30 -30 0>, angularTarget=<0 0 0>}
      localJointOffset=<0 0 -0.15>
     }]
    relativeSpaceBone {boneName=root}
   }
   n17 = Anim Dynamics - Bone: mstr_hay_06(ComponentPose=@n16, Alpha=0.25, ExternalForce="0,0,0") {
    angularDampingOverride=0.96
    angularSpringConstant=45
    bAngularSpring=true
    bGravityOverrideInSimSpace=true
    bOverrideAngularDamping=true
    gravityScale=0.3
    lODThreshold=1
    linearDampingOverride=0.97
    numSolverIterationsPostUpdate=4
    numSolverIterationsPreUpdate=16
    simulationSpace=BoneRelative
    alphaBoolBlend {blendOption=Linear}
    boundBone {boneName=mstr_hay_06}
    componentLinearAccScale=<1 1 1>
    componentLinearVelScale=<1 1 1>
    physicsBodyDefinitions [{
      sphereCollisionRadius=0
      boundBone {boneName=mstr_hay_06}
      boxExtents=<5 5 5>
      constraintSetup {angularTargetAxis=AxisZ, twistAxis=AxisZ, angularLimitsMax=<30 30 0>, angularLimitsMin=<-30 -30 0>, angularTarget=<0 0 0>}
      localJointOffset=<0 0 -0.25>
     }]
    relativeSpaceBone {boneName=root}
   }
   n18 = Anim Dynamics - Bone: mstr_hay_12(ComponentPose=@n17, Alpha=0.2, ExternalForce="0,0,0") {
    angularDampingOverride=0.96
    angularSpringConstant=20
    bAngularSpring=true
    bGravityOverrideInSimSpace=true
    bOverrideAngularDamping=true
    gravityScale=0.35
    lODThreshold=1
    linearDampingOverride=0.97
    numSolverIterationsPostUpdate=4
    numSolverIterationsPreUpdate=16
    simulationSpace=BoneRelative
    alphaBoolBlend {blendOption=Linear}
    boundBone {boneName=mstr_hay_12}
    componentLinearAccScale=<0.75 0.75 0.75>
    componentLinearVelScale=<0.75 0.75 0.75>
    physicsBodyDefinitions [{
      sphereCollisionRadius=0
      boundBone {boneName=mstr_hay_12}
      boxExtents=<5 5 5>
      constraintSetup {angularTargetAxis=AxisZ, twistAxis=AxisZ, angularLimitsMax=<30 30 0>, angularLimitsMin=<-30 -30 0>, angularTarget=<0 0 0>}
      localJointOffset=<0 0 -0.45>
     }]
    relativeSpaceBone {boneName=root}
   }
   n19 = Anim Dynamics - Bone: mstr_hay_11(ComponentPose=@n18, Alpha=0.125, ExternalForce="0,0,0") {
    angularDampingOverride=0.96
    angularSpringConstant=40
    bAngularSpring=true
    bGravityOverrideInSimSpace=true
    bOverrideAngularDamping=true
    gravityScale=0.3
    lODThreshold=1
    linearDampingOverride=0.97
    numSolverIterationsPostUpdate=4
    numSolverIterationsPreUpdate=16
    simulationSpace=BoneRelative
    alphaBoolBlend {blendOption=Linear}
    boundBone {boneName=mstr_hay_11}
    componentLinearAccScale=<0.8 0.8 0.8>
    componentLinearVelScale=<0.8 0.8 0.8>
    physicsBodyDefinitions [{
      sphereCollisionRadius=0
      boundBone {boneName=mstr_hay_11}
      boxExtents=<5 5 5>
      constraintSetup {angularTargetAxis=AxisZ, twistAxis=AxisZ, angularLimitsMax=<30 30 0>, angularLimitsMin=<-30 -30 0>, angularTarget=<0 0 0>}
      localJointOffset=<0 0 -0.4>
     }]
    relativeSpaceBone {boneName=root}
   }
   n20 = Anim Dynamics - Bone: mstr_hay_10(ComponentPose=@n19, Alpha=0.125, ExternalForce="0,0,0") {
    angularDampingOverride=0.96
    angularSpringConstant=35
    bAngularSpring=true
    bGravityOverrideInSimSpace=true
    bOverrideAngularDamping=true
    gravityScale=0.25
    lODThreshold=1
    linearDampingOverride=0.97
    numSolverIterationsPostUpdate=4
    numSolverIterationsPreUpdate=16
    simulationSpace=BoneRelative
    alphaBoolBlend {blendOption=Linear}
    boundBone {boneName=mstr_hay_10}
    componentLinearAccScale=<0.95 0.95 0.95>
    componentLinearVelScale=<0.95 0.95 0.95>
    physicsBodyDefinitions [{
      sphereCollisionRadius=0
      boundBone {boneName=mstr_hay_10}
      boxExtents=<5 5 5>
      constraintSetup {angularTargetAxis=AxisZ, twistAxis=AxisZ, angularLimitsMax=<30 30 0>, angularLimitsMin=<-30 -30 0>, angularTarget=<0 0 0>}
      localJointOffset=<0 0 -0.25>
     }]
    relativeSpaceBone {boneName=root}
   }
   n21 = Anim Dynamics - Bone: mstr_hay_09(ComponentPose=@n20, Alpha=0.25, ExternalForce="0,0,0") {
    angularDampingOverride=0.96
    angularSpringConstant=20
    bAngularSpring=true
    bGravityOverrideInSimSpace=true
    bOverrideAngularDamping=true
    gravityScale=0.4
    lODThreshold=1
    linearDampingOverride=0.97
    numSolverIterationsPostUpdate=4
    numSolverIterationsPreUpdate=16
    simulationSpace=BoneRelative
    alphaBoolBlend {blendOption=Linear}
    boundBone {boneName=mstr_hay_09}
    componentLinearAccScale=<0.9 0.9 0.9>
    componentLinearVelScale=<0.9 0.9 0.9>
    physicsBodyDefinitions [{
      sphereCollisionRadius=0
      boundBone {boneName=mstr_hay_09}
      boxExtents=<5 5 5>
      constraintSetup {angularTargetAxis=AxisZ, twistAxis=AxisZ, angularLimitsMax=<30 30 0>, angularLimitsMin=<-30 -30 0>, angularTarget=<0 0 0>}
      localJointOffset=<0 0 -0.2>
     }]
    relativeSpaceBone {boneName=root}
   }
   n22 = Anim Dynamics - Bone: mstr_hay_08(ComponentPose=@n21, Alpha=0.17, ExternalForce="0,0,0") {
    angularDampingOverride=0.96
    angularSpringConstant=40
    bAngularSpring=true
    bGravityOverrideInSimSpace=true
    bOverrideAngularDamping=true
    gravityScale=0.35
    lODThreshold=1
    linearDampingOverride=0.97
    numSolverIterationsPostUpdate=4
    numSolverIterationsPreUpdate=16
    simulationSpace=BoneRelative
    alphaBoolBlend {blendOption=Linear}
    boundBone {boneName=mstr_hay_08}
    componentLinearAccScale=<0.7 0.7 0.7>
    componentLinearVelScale=<0.7 0.7 0.7>
    physicsBodyDefinitions [{
      sphereCollisionRadius=0
      boundBone {boneName=mstr_hay_08}
      boxExtents=<5 5 5>
      constraintSetup {angularTargetAxis=AxisZ, twistAxis=AxisZ, angularLimitsMax=<30 30 0>, angularLimitsMin=<-30 -30 0>, angularTarget=<0 0 0>}
      localJointOffset=<0 0 -0.3>
     }]
    relativeSpaceBone {boneName=root}
   }
   n23 = Anim Dynamics - Bone: mstr_hay_07(ComponentPose=@n22, Alpha=0.15, ExternalForce="0,0,0") {
    angularDampingOverride=0.96
    angularSpringConstant=35
    bAngularSpring=true
    bGravityOverrideInSimSpace=true
    bOverrideAngularDamping=true
    gravityScale=0.45
    lODThreshold=1
    linearDampingOverride=0.97
    numSolverIterationsPostUpdate=4
    numSolverIterationsPreUpdate=16
    simulationSpace=BoneRelative
    alphaBoolBlend {blendOption=Linear}
    boundBone {boneName=mstr_hay_07}
    componentLinearAccScale=<1 1 1>
    componentLinearVelScale=<1 1 1>
    physicsBodyDefinitions [{
      sphereCollisionRadius=0
      boundBone {boneName=mstr_hay_07}
      boxExtents=<5 5 5>
      constraintSetup {angularTargetAxis=AxisZ, twistAxis=AxisZ, angularLimitsMax=<30 30 0>, angularLimitsMin=<-30 -30 0>, angularTarget=<0 0 0>}
      localJointOffset=<0 0 -0.35>
     }]
    relativeSpaceBone {boneName=root}
   }
   n24 = Anim Dynamics - Bone: mstr_hay_05(ComponentPose=@n23, Alpha=0.16, ExternalForce="0,0,0") {
    angularDampingOverride=0.96
    angularSpringConstant=30
    bAngularSpring=true
    bGravityOverrideInSimSpace=true
    bOverrideAngularDamping=true
    gravityScale=0.25
    lODThreshold=1
    linearDampingOverride=0.97
    numSolverIterationsPostUpdate=4
    numSolverIterationsPreUpdate=16
    simulationSpace=BoneRelative
    alphaBoolBlend {blendOption=Linear}
    boundBone {boneName=mstr_hay_05}
    componentLinearAccScale=<0.85 0.85 0.85>
    componentLinearVelScale=<0.85 0.85 0.85>
    physicsBodyDefinitions [{
      sphereCollisionRadius=0
      boundBone {boneName=mstr_hay_05}
      boxExtents=<5 5 5>
      constraintSetup {angularTargetAxis=AxisZ, twistAxis=AxisZ, angularLimitsMax=<30 30 0>, angularLimitsMin=<-30 -30 0>, angularTarget=<0 0 0>}
      localJointOffset=<0 0 -0.17>
     }]
    relativeSpaceBone {boneName=root}
   }
   n25 = Anim Dynamics - Bone: mstr_hay_15(ComponentPose=@n24, Alpha=0.25, ExternalForce="0,0,0") {
    angularDampingOverride=0.96
    angularSpringConstant=80
    bAngularSpring=true
    bGravityOverrideInSimSpace=true
    bOverrideAngularDamping=true
    gravityScale=0.2
    lODThreshold=1
    linearDampingOverride=0.97
    numSolverIterationsPostUpdate=4
    numSolverIterationsPreUpdate=16
    simulationSpace=BoneRelative
    alphaBoolBlend {blendOption=Linear}
    boundBone {boneName=mstr_hay_15}
    componentLinearAccScale=<0.95 0.95 0.95>
    componentLinearVelScale=<0.95 0.95 0.95>
    physicsBodyDefinitions [{
      sphereCollisionRadius=0
      boundBone {boneName=mstr_hay_15}
      boxExtents=<5 5 5>
      constraintSetup {angularTargetAxis=AxisZ, twistAxis=AxisZ, angularLimitsMax=<30 30 0>, angularLimitsMin=<-30 -30 0>, angularTarget=<0 0 0>}
      localJointOffset=<0 0 -0.1>
     }]
    relativeSpaceBone {boneName=root}
   }
   n26 = Anim Dynamics - Bone: mstr_hay_16(ComponentPose=@n25, Alpha=0.15, ExternalForce="0,0,0") {
    angularDampingOverride=0.96
    angularSpringConstant=45
    bAngularSpring=true
    bGravityOverrideInSimSpace=true
    bOverrideAngularDamping=true
    gravityScale=0.5
    lODThreshold=1
    linearDampingOverride=0.97
    numSolverIterationsPostUpdate=4
    numSolverIterationsPreUpdate=16
    simulationSpace=BoneRelative
    alphaBoolBlend {blendOption=Linear}
    boundBone {boneName=mstr_hay_16}
    componentLinearAccScale=<0.9 0.9 0.9>
    componentLinearVelScale=<0.9 0.9 0.9>
    physicsBodyDefinitions [{
      sphereCollisionRadius=0
      boundBone {boneName=mstr_hay_16}
      boxExtents=<5 5 5>
      constraintSetup {angularTargetAxis=AxisZ, twistAxis=AxisZ, angularLimitsMax=<30 30 0>, angularLimitsMin=<-30 -30 0>, angularTarget=<0 0 0>}
      localJointOffset=<0 0 -0.3>
     }]
    relativeSpaceBone {boneName=root}
   }
   n27 = Anim Dynamics - Bone: mstr_hay_17(ComponentPose=@n26, Alpha=0.217, ExternalForce="0,0,0") {
    angularDampingOverride=0.96
    angularSpringConstant=20
    bAngularSpring=true
    bGravityOverrideInSimSpace=true
    bOverrideAngularDamping=true
    gravityScale=0.35
    lODThreshold=1
    linearDampingOverride=0.97
    numSolverIterationsPostUpdate=4
    numSolverIterationsPreUpdate=16
    simulationSpace=BoneRelative
    alphaBoolBlend {blendOption=Linear}
    boundBone {boneName=mstr_hay_17}
    componentLinearAccScale=<0.8 0.8 0.8>
    componentLinearVelScale=<0.8 0.8 0.8>
    physicsBodyDefinitions [{
      sphereCollisionRadius=0
      boundBone {boneName=mstr_hay_17}
      boxExtents=<5 5 5>
      constraintSetup {angularTargetAxis=AxisZ, twistAxis=AxisZ, angularLimitsMax=<30 30 0>, angularLimitsMin=<-30 -30 0>, angularTarget=<0 0 0>}
      localJointOffset=<0 0 -0.18>
     }]
    relativeSpaceBone {boneName=root}
   }
   n28 = Anim Dynamics - Bone: mstr_hay_13(ComponentPose=@n27, Alpha=0.1, ExternalForce="0,0,0") {
    angularDampingOverride=0.96
    angularSpringConstant=30
    bAngularSpring=true
    bGravityOverrideInSimSpace=true
    bOverrideAngularDamping=true
    gravityScale=0.5
    lODThreshold=1
    linearDampingOverride=0.97
    numSolverIterationsPostUpdate=4
    numSolverIterationsPreUpdate=16
    simulationSpace=BoneRelative
    alphaBoolBlend {blendOption=Linear}
    boundBone {boneName=mstr_hay_13}
    componentLinearAccScale=<0.9 0.9 0.9>
    componentLinearVelScale=<0.9 0.9 0.9>
    physicsBodyDefinitions [{
      sphereCollisionRadius=0
      boundBone {boneName=mstr_hay_13}
      boxExtents=<5 5 5>
      constraintSetup {angularTargetAxis=AxisZ, twistAxis=AxisZ, angularLimitsMax=<30 30 0>, angularLimitsMin=<-30 -30 0>, angularTarget=<0 0 0>}
      localJointOffset=<0 0 -0.2>
     }]
    relativeSpaceBone {boneName=root}
   }
   n29 = Anim Dynamics - Bone: mstr_hay_14(ComponentPose=@n28, Alpha=0.15, ExternalForce="0,0,0") {
    angularDampingOverride=0.96
    angularSpringConstant=46
    bAngularSpring=true
    bGravityOverrideInSimSpace=true
    bOverrideAngularDamping=true
    gravityScale=0.3
    lODThreshold=1
    linearDampingOverride=0.97
    numSolverIterationsPostUpdate=4
    numSolverIterationsPreUpdate=16
    simulationSpace=BoneRelative
    alphaBoolBlend {blendOption=Linear}
    boundBone {boneName=mstr_hay_14}
    componentLinearAccScale=<0.95 0.95 0.95>
    componentLinearVelScale=<0.95 0.95 0.95>
    physicsBodyDefinitions [{
      sphereCollisionRadius=0
      boundBone {boneName=mstr_hay_14}
      boxExtents=<5 5 5>
      constraintSetup {angularTargetAxis=AxisZ, twistAxis=AxisZ, angularLimitsMax=<30 30 0>, angularLimitsMin=<-30 -30 0>, angularTarget=<0 0 0>}
      localJointOffset=<0 0 -0.5>
     }]
    relativeSpaceBone {boneName=root}
   }
   n30 = Anim Dynamics - Bone: mstr_hay_18(ComponentPose=@n29, Alpha=0.215, ExternalForce="0,0,0") {
    angularDampingOverride=0.96
    angularSpringConstant=60
    bAngularSpring=true
    bGravityOverrideInSimSpace=true
    bOverrideAngularDamping=true
    gravityScale=0.4
    lODThreshold=1
    linearDampingOverride=0.97
    numSolverIterationsPostUpdate=4
    numSolverIterationsPreUpdate=16
    simulationSpace=BoneRelative
    alphaBoolBlend {blendOption=Linear}
    boundBone {boneName=mstr_hay_18}
    componentLinearAccScale=<0.9 0.9 0.9>
    componentLinearVelScale=<0.9 0.9 0.9>
    physicsBodyDefinitions [{
      sphereCollisionRadius=0
      boundBone {boneName=mstr_hay_18}
      boxExtents=<5 5 5>
      constraintSetup {angularTargetAxis=AxisZ, twistAxis=AxisZ, angularLimitsMax=<60 0 0>, angularLimitsMin=<-60 0 0>, angularTarget=<0 0 0>}
      localJointOffset=<0 0 -0.2>
     }]
    relativeSpaceBone {boneName=root}
   }
   n31 = Anim Dynamics - Bone: mstr_hay_02(ComponentPose=@n30, Alpha=0.125, ExternalForce="0,0,0") {
    angularDampingOverride=0.96
    angularSpringConstant=20
    bAngularSpring=true
    bGravityOverrideInSimSpace=true
    bOverrideAngularDamping=true
    gravityScale=0.2
    lODThreshold=1
    linearDampingOverride=0.97
    numSolverIterationsPostUpdate=4
    numSolverIterationsPreUpdate=16
    simulationSpace=BoneRelative
    alphaBoolBlend {blendOption=Linear}
    boundBone {boneName=mstr_hay_02}
    componentLinearAccScale=<0.7 0.7 0.7>
    componentLinearVelScale=<0.7 0.7 0.7>
    physicsBodyDefinitions [{
      sphereCollisionRadius=0
      boundBone {boneName=mstr_hay_02}
      boxExtents=<5 5 5>
      constraintSetup {angularTargetAxis=AxisZ, twistAxis=AxisZ, angularLimitsMax=<30 30 0>, angularLimitsMin=<-30 -30 0>, angularTarget=<0 0 0>}
      localJointOffset=<0 0 -0.2>
     }]
    relativeSpaceBone {boneName=root}
   }
   n32 = Anim Dynamics - Bone: mstr_hay_010(ComponentPose=@n31, Alpha=0.1, ExternalForce="0,0,0") {
    angularDampingOverride=0.96
    angularSpringConstant=35
    bAngularSpring=true
    bGravityOverrideInSimSpace=true
    bOverrideAngularDamping=true
    gravityScale=0.2
    lODThreshold=1
    linearDampingOverride=0.97
    numSolverIterationsPostUpdate=4
    numSolverIterationsPreUpdate=16
    simulationSpace=BoneRelative
    alphaBoolBlend {blendOption=Linear}
    boundBone {boneName=mstr_hay_010}
    componentLinearAccScale=<0.65 0.65 0.65>
    componentLinearVelScale=<0.65 0.65 0.65>
    physicsBodyDefinitions [{
      sphereCollisionRadius=0
      boundBone {boneName=mstr_hay_010}
      boxExtents=<5 5 5>
      constraintSetup {angularTargetAxis=AxisZ, twistAxis=AxisZ, angularLimitsMax=<5 60 0>, angularLimitsMin=<-5 -60 0>, angularTarget=<0 0 0>}
      localJointOffset=<0 0 -0.3>
     }]
    relativeSpaceBone {boneName=root}
   }
   n33 = Anim Dynamics - Bone: mstr_hay_03(ComponentPose=@n32, Alpha=0.1, ExternalForce="0,0,0") {
    angularDampingOverride=0.96
    angularSpringConstant=40
    bAngularSpring=true
    bGravityOverrideInSimSpace=true
    bOverrideAngularDamping=true
    gravityScale=0.35
    lODThreshold=1
    linearDampingOverride=0.97
    numSolverIterationsPostUpdate=4
    numSolverIterationsPreUpdate=16
    simulationSpace=BoneRelative
    alphaBoolBlend {blendOption=Linear}
    boundBone {boneName=mstr_hay_03}
    componentLinearAccScale=<1 1 1>
    componentLinearVelScale=<1 1 1>
    physicsBodyDefinitions [{
      sphereCollisionRadius=0
      boundBone {boneName=mstr_hay_03}
      boxExtents=<5 5 5>
      constraintSetup {angularTargetAxis=AxisZ, twistAxis=AxisZ, angularLimitsMax=<5 60 0>, angularLimitsMin=<-15 -60 0>, angularTarget=<0 0 0>}
      localJointOffset=<0 0 -0.1>
     }]
    relativeSpaceBone {boneName=root}
   }
   n34 = Anim Dynamics - Bone: mstr_hay_04(ComponentPose=@n33, Alpha=0.1, ExternalForce="0,0,0") {
    angularDampingOverride=0.96
    angularSpringConstant=60
    bAngularSpring=true
    bGravityOverrideInSimSpace=true
    bOverrideAngularDamping=true
    gravityScale=0.15
    lODThreshold=1
    linearDampingOverride=0.97
    numSolverIterationsPostUpdate=4
    numSolverIterationsPreUpdate=16
    simulationSpace=BoneRelative
    alphaBoolBlend {blendOption=Linear}
    boundBone {boneName=mstr_hay_04}
    componentLinearAccScale=<0.4 0.4 0.4>
    componentLinearVelScale=<0.4 0.4 0.4>
    physicsBodyDefinitions [{
      sphereCollisionRadius=0
      boundBone {boneName=mstr_hay_04}
      boxExtents=<5 5 5>
      constraintSetup {angularTargetAxis=AxisZ, twistAxis=AxisZ, angularLimitsMax=<30 30 0>, angularLimitsMin=<-30 -30 0>, angularTarget=<0 0 0>}
      localJointOffset=<0 0 -1>
     }]
    relativeSpaceBone {boneName=root}
   }
   Output Pose(Result=Component To Local(ComponentPose=@n34)) {layerGroup=None}
  ]}
 EventGraph {nodes [
   TryGetPawnOwner()
   Event BlueprintUpdateAnimation()
  ]}
]
variables []
