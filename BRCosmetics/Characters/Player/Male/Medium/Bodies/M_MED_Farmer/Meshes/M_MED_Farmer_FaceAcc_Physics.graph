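# Physics asset for the face accessory mesh of the male medium Farmer player cosmetic, covering a chick on the hat.
# It holds two zero-length capsules, effectively spheres, on the dyn_hat_chick and head bones, with the chick's shape offset along Z.
# A single constraint joins the chick bone to the head, with a zero limit on swing 1 and equal angular limits on swing 2 and twist.
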PhysicsAsset /BRCosmetics/Characters/Player/Male/Medium/Bodies/M_MED_Farmer/Meshes/M_MED_Farmer_FaceAcc_Physics
bodies {dyn_hat_chick [{shapeType=Capsule, length=0, radius=5, center=<0 0 5>}], head [{shapeType=Capsule, length=0, radius=5}]}
constraints [{bone1Name=dyn_hat_chick, bone2Name=head, swing1LimitDegrees=0, swing2LimitDegrees=45, twistLimitDegrees=45}]
