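# Animation Blueprint for the Farmer male medium character's face accessory, adding secondary physics motion to the attached cosmetic mesh.
# It copies the pose from the attached parent mesh, then runs a RigidBody simulation in spine_03 space using the accessory's physics asset.
# The simulation blends in partially, with a lower alpha while bIsTargeting is true.
# An AnimDynamics node for dyn_hat_chick is left unconnected, and the event graph has no logic.
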AnimBlueprint /BRCosmetics/Characters/Player/Male/Medium/Bodies/M_MED_Farmer/Meshes/M_MED_Farmer_FaceAcc_AnimBP
target_skeleton=&/M_MED_Farmer_FaceAcc_Skeleton
graphs [
 AnimGraph {nodes [
   Anim Dynamics - Bone: dyn_hat_chick(Alpha=0.1, ExternalForce="0,0,0") {
    angularDampingOverride=0.97
    angularSpringConstant=20
    bGravityOverrideInSimSpace=true
    gravityScale=0.3
    lODThreshold=1
    linearDampingOverride=0.97
    numSolverIterationsPostUpdate=4
    numSolverIterationsPreUpdate=16
    simulationSpace=BoneRelative
    alphaBoolBlend {blendOption=Linear}
    boundBone {boneName=dyn_hat_chick}
    componentLinearAccScale=<1 1 1>
    componentLinearVelScale=<1 1 1>
    physicsBodyDefinitions [{
      sphereCollisionRadius=0
      boundBone {boneName=dyn_hat_chick}
      boxExtents=<5 5 5>
      constraintSetup {angularTargetAxis=AxisZ, twistAxis=AxisZ, angularLimitsMax=<60 60 0>, angularLimitsMin=<-60 -60 0>, angularTarget=<0 0 0>}
      localJointOffset=<0 0 -0.25>
     }]
    relativeSpaceBone {boneName=pelvis}
   }
   n1 = RigidBody(
    ComponentPose=Local To Component(LocalPose=Copy Pose From Mesh() {bUseAttachedParent=true})
    Alpha=SelectFloat(A=0.25, B=0.5, bPickA=Get bIsTargeting())
   ) {
    bOverrideWorldGravity=true
    overridePhysicsAsset=&/M_MED_Farmer_FaceAcc_Physics
    simulationSpace=BaseBoneSpace
    alphaBoolBlend {blendOption=Linear}
    baseBoneRef {boneName=spine_03}
    componentLinearAccScale=<1 1 1>
    componentLinearVelScale=<1 1 1>
    overrideWorldGravity {z=-980}
   }
   Output Pose(Result=Component To Local(ComponentPose=@n1)) {layerGroup=None}
  ]}
 EventGraph {nodes [Event BlueprintUpdateAnimation()]}
]
variables []
